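; Parent gameplay ability that automatically revives an AI-controlled player pawn from DBNO (down but not out) after a configurable delay.
; On activation it caches the avatar pawn and binds to its DBNO and death events, ending if the cast fails.
; Entering DBNO starts a timer that, unless death was flagged, revives the pawn and optionally restores full health.
; Ending the ability unbinds both events but leaves any pending timer running.
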(event Ability|EventActivateAbility
  (Variables|InternalData|SetPawnIsDead false)
  (bind _outfortplayerpawn (CallFunction|GetAvatarasValidFortPlayerPawn)
    (:ValidFortPlayerPawn
      (bind _output_get (Variables|InternalData|SetAIPawn _outfortplayerpawn))
      (EventDispatchers|BindEventtoOnDBNOStateChanged _output_get (EventDispatchers|CreateEvent :!event Custom|AIPawnDBNOStateChanged))
      (EventDispatchers|BindEventtoOnDied (Variables|InternalData|GetAIPawn) (EventDispatchers|CreateEvent :!event Custom|AIPawnDied)))
    (:AvatarCastFailed
      (Ability|EndAbility))))

(event Custom|AIPawnDBNOStateChanged
  (params (FortPawn :type Fort%20Pawn%20Object%20Reference) (bInIsDBNO :type Boolean))
  (if bInIsDBNO
    (bind _returnvalue (Utilities|Time|SetTimerbyEvent (EventDispatchers|CreateEvent :!event Custom|DBNOTimerElapsed) (ScalableFloat|GetValueatLevel (Variables|Default|GetDelayBeforeRevivefromDBNO))))
    (Variables|InternalData|SetHNDL_DBNOTimer _returnvalue)))

(event Custom|DBNOTimerElapsed
  (if (not (Variables|InternalData|GetPawnIsDead))
    (FortPlayerPawn|ManualRevivefromDBNO (Variables|InternalData|GetAIPawn) 0)
    (Utilities|FlowControl|Sequence
      (:then_0
        _)
      (:then_1
        (if (ScalableFloat|GetValueasBool (Variables|Default|GetFullyHealOnDBNORevive))
          (FortPawn|SetHealth (Variables|InternalData|GetAIPawn) (FortPawn|GetMaxHealth (Variables|InternalData|GetAIPawn)))))))
  (Utilities|Time|ClearandInvalidateTimerbyHandle (Variables|InternalData|GetHNDL_DBNOTimer)))

(event Custom|AIPawnDied
  (params (DamagedActor :type Actor%20Object%20Reference) (Damage :type Float%20%28single-precision%29) (InstigatedBy :type Controller%20Object%20Reference) (DamageCauser :type Actor%20Object%20Reference) (HitLocation :type Vector) (FHitComponent :type Primitive%20Component%20Object%20Reference) (BoneName :type Name) (Momentum :type Vector))
  (Variables|InternalData|SetPawnIsDead true))

(event Ability|EventOnEndAbility
  (params (bWasCancelled :type Boolean))
  (bind _aipawn (Variables|InternalData|GetAIPawn))
  (Utilities|IsValid _aipawn
    (:Is%20Valid
      (EventDispatchers|UnbindEventfromOnDBNOStateChanged _aipawn (EventDispatchers|CreateEvent :!event Custom|AIPawnDBNOStateChanged))
      (EventDispatchers|UnbindEventfromOnDied _aipawn (EventDispatchers|CreateEvent :!event Custom|AIPawnDied)))))
